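# NS_CarSmoking is a Niagara system in the car destruction effects folder, named as a smoke effect for cars.
# It holds a single emitter, LW_Explosion_Main, and uses only the SystemState module.
# No renderers, data interfaces, dynamic inputs or user variables are listed, so the visible output of the effect is not shown.
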NiagaraSystem /Game/Effects/Fort_Effects/World/Destruction/Cars/NS_CarSmoking
dependencies {usedDataInterfaces [], usedDynamicInputs [], usedModules [&SystemState]}
emitters {LW_Explosion_Main {rendererClasses [], renderers []}}
summary {systemName=NS_CarSmoking, emitters [{emitterName=LW_Explosion_Main, rendererClasses []}], userVariables []}
