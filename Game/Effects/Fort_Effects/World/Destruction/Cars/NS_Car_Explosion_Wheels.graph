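# Niagara system for car explosions that throws four wheel meshes outward and upward from the corners of the car.
# One mesh emitter bursts four short-lived particles, with a partition index selecting each one's corner position, orientation and randomized sideways-plus-upward launch velocity.
# A constant downward acceleration pulls them into arcs while a random rotation vector spins them.
# Curves over life shrink the meshes and fade color alpha to zero.
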NiagaraSystem /Game/Effects/Fort_Effects/World/Destruction/Cars/NS_Car_Explosion_Wheels
dependencies {
 usedDataInterfaces [&NiagaraDataInterfaceCurve, &NiagaraDataInterfaceVectorCurve]
 usedDynamicInputs []
 usedModules [&SystemState]
}
emitters {Particle_Emitter {
  emitterUpdateScript [
   &EmitterState()
   &SpawnBurst_Instantaneous(Spawn Count=4, Use Spawn Probability=false, UseLoopCountLimit=false)
  ]
  particleSpawnScript [
   &PartitionParticles(Number Of Partitions=4)
   /SetVariables(Particles.MeshOrientation=&MakeQuatFromAxisAngle(
     Axis=&MergeHelper_VectorSwitch(0=<0 0 0>, 1=<0 0 0.5>, 2=<0 0 0>, 3=<0 0 0.5>, Index=$Particles.PartitionParticles.Partition, Num=::NewEnumerator2)
     AngleInDegrees=90
    )) {isSetParameters=true}
   /SetVariables(Particles.Acceleration:Vector3f=<0 0 -1000>) {isSetParameters=true}
   /SetVariables(Particles.SpawnRotationVector=&RandomRangeVector(Minimum=<0 0 0>, Maximum=<1 1 1>)) {isSetParameters=true}
   &InitializeParticle(
    Lifetime=1.2
    Mesh Scale=<0.6 0.6 0.6>
    Mesh Scale Mode=::NewEnumerator4
    Position=&ConvertVectorToPosition(
     Input Position=&MergeHelper_VectorSwitch(0=<40 130 -100>, 1=<-40 130 -100>, 2=<40 -130 -100>, 3=<-40 -130 -100>, Index=$Particles.PartitionParticles.Partition, Num=::NewEnumerator2)
     Conversion Operation:ENiagara_LWCConvertVecToPos=::NewEnumerator0
    )
    Position Mode=::NewEnumerator1
    UsePositionOffset=false
    Write Material Random=true
   )
   &AddVelocity(Velocity=&MergeHelper_VectorSwitch(
     0=&RandomRangeVector(Minimum=<200 -100 400>, Maximum=<300 100 600>)
     1=&RandomRangeVector(Minimum=<-200 -100 400>, Maximum=<-300 100 600>)
     2=&RandomRangeVector(Minimum=<200 -100 400>, Maximum=<300 100 600>)
     3=&RandomRangeVector(Minimum=<-200 -100 400>, Maximum=<-300 100 600>)
     Index=$Particles.PartitionParticles.Partition
     Num=::NewEnumerator2
    ))
  ]
  particleUpdateScript [
   &ParticleState()
   &ScaleMeshSize(Scale Factor=&VectorFromFloat(Value=&FloatFromCurve(FloatCurve=DataInterface() {propertyValues {Curve {keys [<at 0 1>, {arriveTangent=-1, interpMode=RCIM_Cubic, leaveTangent=-1, time=0.6, value=1}, <at 1 0>]}}})))
   &UpdateMeshOrientation(Rotation Vector=$Particles.SpawnRotationVector)
   &Color(Color=&MakeLinearColorFromVectorAndFloat(
     "Vector (RGB)"=&VectorFromCurve(VectorCurve=DataInterface() {propertyValues {XCurve {keys [<at 0 1>,<at 1 1>]}, YCurve {keys [<at 0 1>,<at 1 1>]}, ZCurve {keys [<at 0 1>,<at 1 1>]}}})
     "Float (Alpha)"=&ClampFloat(Float=&FloatFromCurve(FloatCurve=DataInterface() {propertyValues {Curve {keys [<at 0 1>,<at 1 0>]}}}))
    ))
   &AccelerationForce(Acceleration=$Particles.Acceleration)
   &SolveForcesAndVelocity(Clamp Velocity=false, Limit Acceleration=false)
  ]
  rendererClasses [&NiagaraMeshRendererProperties]
  renderers [{rendererClass=&NiagaraMeshRendererProperties, rendererIndex=0}]
 }}
summary {
 systemName=NS_Car_Explosion_Wheels
 emitters [{emitterName=Particle_Emitter, rendererClasses [&NiagaraMeshRendererProperties]}]
 userVariables []
}
